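# Niagara system for the equipped effect of the StaminaAlaska consumable, drawing meshes that trail and deform with owner movement.
# A burst-spawned particle runs FollowingDeform, passing moving average, stretch distance and constraint axis to materials through dynamic parameters.
# Age-driven curves animate the intro stretch, plane constraint, scale and fade, with User.Deployed offsetting most curve inputs toward their end values.
# User.IsFrontEnd switches lighting, fade and deform behavior, replacing time-of-day material values with fixed ones.
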NiagaraSystem /Game/Athena/Items/Consumables/StaminaAlaska/FX/NS_StaminaAlaska_Equipped
dependencies {
 usedDataInterfaces [&NiagaraDataInterfaceCurve, &NiagaraDataInterfaceArrayFloat]
 usedDynamicInputs []
 usedModules [&SystemState, &Timeline, &InterpolateOverTime]
}
emitters {FN_FollowingDeformTemplate {
  emitterUpdateScript [
   &EmitterState(
    Inactive Response=::NewEnumerator1
    Life Cycle Mode=::NewEnumerator1
    Loop Behavior=::NewEnumerator1
    Loop Duration Mode=::NewEnumerator1
    UseLoopDelay=false
   )
   &SpawnBurst_Instantaneous(Use Spawn Probability=false, UseLoopCountLimit=false)
   /SetVariables(
    Emitter.DepthFadeDistance=1
    Emitter.FELight:LinearColor=&MakeCustomLinearColorFromBool(Bool=$User.IsFrontEnd, True Color=<rgba 0 0 1 1>, False Color=<rgba 0 0 0 0>)
   ) {isSetParameters=true}
  ]
  particleSpawnScript [
   &InitializeParticle(
    Color Mode=::NewEnumerator0
    Lifetime=5
    Mesh Scale=<1.2 1.1 1>
    Mesh Scale Mode=::NewEnumerator4
    Position Mode=::NewEnumerator0
    Sprite Size Mode=::NewEnumerator3
    Uniform Sprite Size=100
    Write Material Random=true
   )
   &InitialMeshOrientation(Mesh Orientation Mode=::NewEnumerator2, Use Rotation Vector=false) {enabled=false}
   /SetVariables() {isSetParameters=true}
  ]
  particleUpdateScript [
   &ParticleState(Kill Particles When Lifetime Has Elapsed=false, Let Infinitely Lived Particles Die When Emitter Deactivates=true)
   /SetVariables(Particles.FollowingDeform.Moving Average:Vector3f=&MakeCustomVectorFromBool(
     Bool=&Boolean_OrOperation(Bool1=$User.IsFrontEnd, Bool2=&SetBoolByIntComparison(A=$User.Deployed, B=1, Comparison Type=::NewEnumerator4))
     True Vector=$Particles.FollowingDeform.Moving Average
     False Vector=&Lerp_Vector(
      A=&Add_Vector(
       A=Linked() {linkedVariable {name=Engine.Owner.Position, type=NiagaraPosition}}
       B=&TransformVector(Vector=&Multiply_Vector(A=<-10 8 10>, B=$Engine.Owner.Scale), Destination Space=::World)
      )
      B=$Particles.FollowingDeform.Moving Average
      Alpha=&FloatFromCurve(
       FloatCurve=DataInterface() {propertyValues {Curve {keys [<at 1 0>,<at 2 1>]}}}
       CurveIndex=&Add_Float(A=$Particles.Age, B=&Multiply_FloatByInt(Float=1, Integer=$User.Deployed))
       Scale Curve=0.8
      )
     )
    )) {isSetParameters=true}
   /Game/Effects/Niagara/Modules/FollowingDeform/FN_FollowingDeform(
    CustomMeshOrientation=false
    Distance=75
    MaxAngle=90
    MaxStretchDistance:NiagaraFloat=&MakeCustomFloatFromBool(
     Bool=&Boolean_OrOperation(Bool1=$User.IsFrontEnd, Bool2=&SetBoolByIntComparison(A=$User.Deployed, B=1, Comparison Type=::NewEnumerator4))
     True Float=25
     False Float=&FloatFromCurve(
      FloatCurve=DataInterface() {propertyValues {Curve {keys [
          {interpMode=RCIM_Cubic, leaveTangent=-1074.205, tangentMode=RCTM_Break, time=0, value=0}
          {arriveTangent=1821.9176, interpMode=RCIM_Cubic, leaveTangent=-5700.3721, tangentMode=RCTM_Break, time=0.75, value=2500}
          {interpMode=RCIM_Cubic, tangentMode=RCTM_Break, time=1.85, value=100}
         ]}}}
      CurveIndex=&Add_Float(A=$Particles.Age, B=&Multiply_FloatByInt(Float=1, Integer=$User.Deployed))
     )
    )
    MeshX+AxisAlignLS:Vector3f=<1 0 0>
    MeshZ+AxisAlignLS:Vector3f=<0 1 0>
    PlaneConstraint=true
    PlaneConstraintStrength:NiagaraFloat=&FloatFromCurve(
     FloatCurve=DataInterface() {propertyValues {Curve {keys [<at 1 0>,<at 1.35 1>]}}}
     CurveIndex=&Add_Float(A=$Particles.Age, B=&Multiply_FloatByInt(Float=100, Integer=$User.Deployed))
    )
    RootPosition:NiagaraPosition=&Add_Vector(
     A=Linked() {linkedVariable {name=Engine.Owner.Position, type=NiagaraPosition}}
     B=&TransformVector(
      Vector=&Multiply_Vector(A=&MakeVector(X=0, Y=5, Z=0), B=$Engine.Owner.Scale)
      Destination Space=::World
     )
    )
    Speed=2
    StretchRecoverSpeed:NiagaraFloat=&FloatFromCurve(
     FloatCurve=DataInterface() {propertyValues {Curve {keys [<at 0.8 0.2>,<at 1.35 2>]}}}
     CurveIndex=&Add_Float(A=$Particles.Age, B=&Multiply_FloatByInt(Float=100, Integer=$User.Deployed))
    )
   )
   &DynamicMaterialParameters(
    Index 0 Param 1=1.5
    Index 0 Param 2=0.2
    Index 0 Param 3=&FloatFromCurve(
     FloatCurve=DataInterface() {propertyValues {Curve {keys [<at 0 -1>,<at 1.45 0>]}}}
     CurveIndex=&Add_Float(A=$Particles.Age, B=&Multiply_FloatByInt(Float=100, Integer=$User.Deployed))
    )
    Index 0 Param 4=&FloatFromCurve(
     FloatCurve=DataInterface() {propertyValues {Curve {keys [<at 0 -1>,<at 1.45 0.5>]}}}
     CurveIndex=&Add_Float(A=$Particles.Age, B=&Multiply_FloatByInt(Float=100, Integer=$User.Deployed))
    )
    Index 1 Param 2=&MakeCustomFloatFromBool(
     Bool=$User.IsFrontEnd
     True Float=0
     False Float=&FNTimeOfDayScalar(Night Scale=0.05, Morning Scale=-0.1, Day Scale=0, Evening Scale=-0.1)
    )
    Index 1 Param 3=&MakeCustomFloatFromBool(
     Bool=$User.IsFrontEnd
     True Float=1.6
     False Float=&FNTimeOfDayScalar(Night Scale=1, Morning Scale=0.3, Evening Scale=0.3)
    )
    Index 2 Param 1=&MakeFloatFromVector(Vector=$Particles.FollowingDeform.Moving Average, Channel=::NewEnumerator0)
    Index 2 Param 2=&MakeFloatFromVector(Vector=$Particles.FollowingDeform.Moving Average, Channel=::NewEnumerator1)
    Index 2 Param 3=&MakeFloatFromVector(Vector=$Particles.FollowingDeform.Moving Average, Channel=::NewEnumerator2)
    Index 2 Param 4=$Particles.FollowingDeform.StretchDistance
    Index 3 Param 1=&MakeFloatFromVector(Vector=$Particles.FollowingDeform.ConstraintAxisWS, Channel=::NewEnumerator0)
    Index 3 Param 2=&MakeFloatFromVector(Vector=$Particles.FollowingDeform.ConstraintAxisWS, Channel=::NewEnumerator1)
    Index 3 Param 3=&MakeFloatFromVector(Vector=$Particles.FollowingDeform.ConstraintAxisWS, Channel=::NewEnumerator2)
    Index 3 Param 4=&Multiply_Float(
     A=&MakeFloatFromVector(Vector=$Engine.Owner.Scale, Channel=::NewEnumerator0)
     B=&FloatFromCurve(
      FloatCurve=DataInterface() {propertyValues {Curve {keys [<at 0 0.5>,<at 1.7 1>]}}}
      CurveIndex=&Add_Float(A=$Particles.Age, B=&Multiply_FloatByInt(Float=100, Integer=$User.Deployed))
     )
    )
    Module.Param0WriteEnabled=true
    Module.Param0WriteEnabled1=false
    Module.Param0WriteEnabled2=true
    Module.Param0WriteEnabled3=true
    Module.Param1WriteEnabled=true
    Module.Param1WriteEnabled1=true
    Module.Param1WriteEnabled2=true
    Module.Param1WriteEnabled3=true
    Module.Param2WriteEnabled=true
    Module.Param2WriteEnabled1=true
    Module.Param2WriteEnabled2=true
    Module.Param2WriteEnabled3=true
    Module.Param3WriteEnabled=true
    Module.Param3WriteEnabled1=false
    Module.Param3WriteEnabled2=true
    Module.Param3WriteEnabled3=true
    Write Parameter Index 1=true
    Write Parameter Index 2=true
    Write Parameter Index 3=true
   )
   &ScaleMeshSize(Scale Factor=$Engine.Owner.Scale)
   &ScaleMeshSize(Scale Factor=&VectorFromFloat(Value=&FloatFromCurve(
      FloatCurve=DataInterface() {propertyValues {Curve {keys [{interpMode=RCIM_Cubic, leaveTangent=1.3187, tangentMode=RCTM_Break, time=0.2, value=0}, {tangentMode=RCTM_User, time=1, value=1}]}}}
      CurveIndex=$Emitter.Age
     ))) {module=ScaleMeshSize001}
   &ScaleMeshSize(Scale Factor=&MakeVector(
     X=&FloatFromCurve(
      FloatCurve=DataInterface() {propertyValues {Curve {keys [
          {interpMode=RCIM_Cubic, leaveTangent=0.7843, tangentMode=RCTM_Break, time=0.4, value=0.2}
          {arriveTangent=0.7835, interpMode=RCIM_Cubic, tangentMode=RCTM_Break, time=1.6, value=1}
         ]}}}
      CurveIndex=&Add_Float(A=$Particles.Age, B=&Multiply_FloatByInt(Float=100, Integer=$User.Deployed))
     )
     Y=&FloatFromCurve(
      FloatCurve=DataInterface() {propertyValues {Curve {keys [
          {interpMode=RCIM_Cubic, leaveTangent=0.7843, tangentMode=RCTM_Break, time=0.4, value=0.2}
          {arriveTangent=0.7835, interpMode=RCIM_Cubic, tangentMode=RCTM_Break, time=1.6, value=1}
         ]}}}
      CurveIndex=&Add_Float(A=$Particles.Age, B=&Multiply_FloatByInt(Float=100, Integer=$User.Deployed))
     )
     Z=1
    )) {enabled=false, module=ScaleMeshSize002}
   &ScaleColor(
    Scale Alpha=&FloatFromCurve(
     FloatCurve=DataInterface() {propertyValues {Curve {keys [<at 0 0>,<at 0.2 1>]}}}
     CurveIndex=&Add_Float(A=$Particles.Age, B=&Multiply_FloatByInt(Float=100, Integer=$User.Deployed))
     Scale Curve=&MakeCustomFloatFromBool(
      Bool=$User.IsFrontEnd
      True Float=&FloatFromCurve(
       FloatCurve=DataInterface() {propertyValues {Curve {keys [<at 1.5 0>,<at 1.6 1>]}}}
       CurveIndex=&Add_Float(A=$Particles.Age, B=&Multiply_FloatByInt(Float=100, Integer=$User.Deployed))
      )
      False Float=1
     )
    )
    ScaleA=true
    ScaleRGB=true
   )
   &ScaleMeshSize(Scale Factor=&VectorFromFloat(Value=0.09)) {module=ScaleMeshSize003}
  ]
  rendererClasses [&NiagaraSpriteRendererProperties, &NiagaraMeshRendererProperties]
  renderers [
   {rendererClass=&NiagaraSpriteRendererProperties}
   {rendererClass=&NiagaraMeshRendererProperties, rendererIndex=1}
   {rendererClass=&NiagaraMeshRendererProperties, rendererIndex=2}
   {rendererClass=&NiagaraMeshRendererProperties, rendererIndex=3}
  ]
 }}
summary {
 systemName=NS_StaminaAlaska_Equipped
 emitters [{
   emitterName=FN_FollowingDeformTemplate
   rendererClasses [&NiagaraSpriteRendererProperties, &NiagaraMeshRendererProperties]
  }]
 userVariables [
  User.Deployed {type=NiagaraInt32}
  User.EquipState {type=NiagaraInt32}
  User.ForwardDot {type=NiagaraFloat, defaultValue=0, description=""}
  User.Hands {
   type=NiagaraDataInterfaceSkeletalMesh
   defaultValue {
    struct=NiagaraExt_VariableValue_DataInterface
    value {dataInterface=&self:NiagaraDataInterfaceSkeletalMesh_1}
   }
  }
  User.IsFrontEnd {type=NiagaraBool}
  User.RibbonIDOffset {type=NiagaraInt32}
  User.SkeletalMesh {
   type=NiagaraDataInterfaceSkeletalMesh
   defaultValue {
    struct=NiagaraExt_VariableValue_DataInterface
    value {dataInterface=&self:NiagaraDataInterfaceSkeletalMesh_0}
   }
  }
 ]
}
